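# Physics asset for the F_MED Motorcyclist backpack cosmetic, giving capsule bodies to its two swords, chains, straps, zipper and circle chain.
# Straps, zipper and short chains hang from backpack_mstr with angular limits, while longer chains link segment by segment from each blade.
# Reverse chain-end bones use angularly free constraints to the pack and several segments, apparently closing loops.
# The right blade is tightly limited, but the left is angularly free.
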PhysicsAsset /BRCosmetics/Accessories/FORT_Backpacks/Backpack_F_MED_Motorcyclist/Meshes/F_MED_Motorcyclist_Pack_Physics
bodies {
 backpack_mstr [{shapeType=Capsule, length=6, radius=5, shapeName=backpack_mstr_capsule, center=<1.7223 -2.0212 -5.9882>, rotation=<rot 15 0 -95>}]
 dyn_circle_chain_01 [{shapeType=Capsule, length=2, radius=1, shapeName=dyn_circle_chain_01_capsule, center=<1 0 0>, rotation=<rot 90 0 0>}]
 dyn_front_strap_base [{shapeType=Capsule, length=2, radius=1, shapeName=dyn_front_strap_base_capsule, center=<1 0 0>, rotation=<rot 90 0 0>}]
 dyn_l_sword_blade_base [{shapeType=Capsule, length=15, radius=2, shapeName=dyn_r_sword_blade_base_capsule, rotation=<rot 90 0 0>}]
 dyn_l_sword_chain_01 [{shapeType=Capsule, length=4, radius=1, shapeName=dyn_l_sword_chain_01_capsule, center=<3.5 0 0>, rotation=<rot 90 0 0>}]
 dyn_l_sword_chain_02 [{shapeType=Capsule, length=4, radius=1, shapeName=dyn_l_sword_chain_02_capsule, center=<3 0 0>, rotation=<rot 90 0 0>}]
 dyn_l_sword_chain_03 [{shapeType=Capsule, length=2, radius=1, shapeName=dyn_l_sword_chain_03_capsule, center=<1.5 0 0>, rotation=<rot 90 0 0>}]
 dyn_l_sword_chain_04 [{shapeType=Capsule, length=4, radius=1, shapeName=dyn_l_sword_chain_03_capsule, center=<3.5 0 0>, rotation=<rot 90 0 0>}]
 dyn_l_sword_chain_05 [{shapeType=Capsule, length=2, radius=1, shapeName=dyn_l_sword_chain_03_capsule}]
 dyn_l_sword_chain_05_reverse [{shapeType=Capsule, length=4, radius=1, shapeName=dyn_l_sword_chain_05_reverse_capsule, center=<3.5 0 0>, rotation=<rot 90 0 0>}]
 dyn_l_sword_chain_06 [{shapeType=Capsule, length=1, radius=1, shapeName=dyn_l_sword_chain_06_capsule, center=<1 0 0>, rotation=<rot 90 0 0>}]
 dyn_l_sword_chain_07 [{shapeType=Capsule, length=1.5, radius=1, shapeName=dyn_l_sword_chain_07_capsule, center=<1.5 0 0>, rotation=<rot 90 0 0>}]
 dyn_l_sword_chain_08 [{shapeType=Capsule, length=1.5, radius=1, shapeName=dyn_l_sword_chain_08_capsule, center=<1.5 0 0>, rotation=<rot 90 0 0>}]
 dyn_l_sword_chain_reverse_base [{shapeType=Capsule, length=2, radius=1, shapeName=dyn_l_sword_chain_reverse_base_capsule}]
 dyn_r_sword_blade_base [{shapeType=Capsule, length=15, radius=2, shapeName=dyn_r_sword_blade_base_capsule, rotation=<rot 90 0 0>}]
 dyn_r_sword_chain_01 [{shapeType=Capsule, length=4, radius=1, shapeName=dyn_r_sword_chain_01_capsule, center=<3.5 0 0>, rotation=<rot 90 0 0>}]
 dyn_r_sword_chain_02 [{shapeType=Capsule, length=4.5, radius=1, shapeName=dyn_r_sword_chain_02_capsule, center=<3.25 0 0>, rotation=<rot 90 0 0>}]
 dyn_r_sword_chain_03 [{shapeType=Capsule, length=1.5, radius=1, shapeName=dyn_r_sword_chain_03_capsule, center=<2 0 0>, rotation=<rot 90 0 0>}]
 dyn_r_sword_chain_04 [{shapeType=Capsule, length=4, radius=1, shapeName=dyn_r_sword_chain_04_capsule, center=<3.5 0 0>, rotation=<rot 90 0 0>}]
 dyn_r_sword_chain_05 [{shapeType=Capsule, length=2, radius=1, shapeName=dyn_r_sword_chain_04_capsule}]
 dyn_r_sword_chain_05_reverse [{shapeType=Capsule, length=4, radius=1, shapeName=dyn_r_sword_chain_05_reverse_capsule, center=<3.5 0 0>, rotation=<rot 90 0 0>}]
 dyn_r_sword_chain_06 [{shapeType=Capsule, length=1, radius=1, shapeName=dyn_r_sword_chain_06_capsule, center=<1 0 0>, rotation=<rot 90 0 0>}]
 dyn_r_sword_chain_07 [{shapeType=Capsule, length=1.5, radius=1, shapeName=dyn_r_sword_chain_07_capsule, center=<1.5 0 0>, rotation=<rot 90 0 0>}]
 dyn_r_sword_chain_08 [{shapeType=Capsule, length=1.5, radius=1, shapeName=dyn_r_sword_chain_08_capsule, center=<1.5 0 0>, rotation=<rot 90 0 0>}]
 dyn_r_sword_chain_09 [{shapeType=Capsule, length=2.5, radius=1.5, shapeName=dyn_r_sword_chain_09_capsule, center=<3.5 0 0>, rotation=<rot 90 0 0>}]
 dyn_r_sword_chain_reverse_base [{shapeType=Capsule, length=2, radius=1, shapeName=dyn_r_sword_chain_reverse_base_capsule}]
 dyn_r_zipper_base [{shapeType=Capsule, length=2, radius=0.5, shapeName=dyn_r_zipper_base_capsule, center=<1 0.3 0>, rotation=<rot 90 0 0>}]
 dyn_top_strap [{shapeType=Capsule, length=2, radius=1, shapeName=dyn_top_strap_capsule, center=<1.5 0 0>, rotation=<rot 90 0 90>}]
}
constraints [
 {bone1Name=dyn_r_sword_chain_01, bone2Name=dyn_r_sword_blade_base, swing1LimitDegrees=60, swing2LimitDegrees=60, twistLimitDegrees=5}
 {bone1Name=dyn_r_sword_chain_02, bone2Name=dyn_r_sword_chain_01, swing1LimitDegrees=60, swing2LimitDegrees=60, twistLimitDegrees=5}
 {bone1Name=dyn_r_sword_chain_03, bone2Name=dyn_r_sword_chain_02, swing1LimitDegrees=60, swing2LimitDegrees=60, twistLimitDegrees=5}
 {bone1Name=dyn_r_sword_chain_09, bone2Name=dyn_r_sword_chain_08, swing1LimitDegrees=30, swing2LimitDegrees=30, twistLimitDegrees=1}
 {bone1Name=dyn_r_sword_chain_07, bone2Name=dyn_r_sword_chain_06, swing1LimitDegrees=30, swing2LimitDegrees=30, twistLimitDegrees=1}
 {bone1Name=dyn_r_sword_chain_08, bone2Name=dyn_r_sword_chain_07, swing1LimitDegrees=30, swing2LimitDegrees=30, twistLimitDegrees=1}
 {bone1Name=dyn_l_sword_chain_03, bone2Name=dyn_l_sword_chain_02, swing1LimitDegrees=60, swing2LimitDegrees=60, twistLimitDegrees=5}
 {bone1Name=dyn_l_sword_chain_07, bone2Name=dyn_l_sword_chain_06, swing1LimitDegrees=30, swing2LimitDegrees=30, twistLimitDegrees=1}
 {bone1Name=dyn_l_sword_chain_08, bone2Name=dyn_l_sword_chain_07, swing1LimitDegrees=30, swing2LimitDegrees=30, twistLimitDegrees=1}
 {bone1Name=dyn_r_sword_blade_base, bone2Name=backpack_mstr, swing1LimitDegrees=5, swing2LimitDegrees=5, twistLimitDegrees=1}
 {bone1Name=dyn_r_sword_chain_06, bone2Name=backpack_mstr, swing1LimitDegrees=30, swing2LimitDegrees=30, twistLimitDegrees=1}
 {bone1Name=dyn_l_sword_chain_06, bone2Name=backpack_mstr, swing1LimitDegrees=30, swing2LimitDegrees=30, twistLimitDegrees=1}
 {bone1Name=dyn_circle_chain_01, bone2Name=backpack_mstr, swing1LimitDegrees=45, swing2LimitDegrees=45, twistLimitDegrees=45}
 {bone1Name=dyn_top_strap, bone2Name=backpack_mstr, swing1LimitDegrees=15, swing2LimitDegrees=35, twistLimitDegrees=1}
 {bone1Name=dyn_front_strap_base, bone2Name=backpack_mstr, swing1LimitDegrees=30, swing2LimitDegrees=30, twistLimitDegrees=1}
 {bone1Name=dyn_r_zipper_base, bone2Name=backpack_mstr, swing1LimitDegrees=30, swing2LimitDegrees=30, twistLimitDegrees=1}
 {bone1Name=dyn_l_sword_chain_01, bone2Name=dyn_l_sword_blade_base, swing1LimitDegrees=60, swing2LimitDegrees=60, twistLimitDegrees=5}
 {bone1Name=dyn_l_sword_chain_02, bone2Name=dyn_l_sword_chain_01, swing1LimitDegrees=60, swing2LimitDegrees=60, twistLimitDegrees=5}
 {bone1Name=dyn_r_sword_chain_04, bone2Name=dyn_r_sword_chain_03, swing1LimitDegrees=60, swing2LimitDegrees=60, twistLimitDegrees=5}
 {bone1Name=dyn_l_sword_chain_04, bone2Name=dyn_l_sword_chain_03, swing1LimitDegrees=60, swing2LimitDegrees=60, twistLimitDegrees=5}
 {
  bone1Name=dyn_l_sword_chain_reverse_base
  bone2Name=backpack_mstr
  swing1LimitDegrees=45
  swing1Motion=Free
  swing2LimitDegrees=45
  swing2Motion=Free
  twistLimitDegrees=45
  twistMotion=Free
 }
 {
  bone1Name=dyn_r_sword_chain_reverse_base
  bone2Name=backpack_mstr
  swing1LimitDegrees=45
  swing1Motion=Free
  swing2LimitDegrees=45
  swing2Motion=Free
  twistLimitDegrees=45
  twistMotion=Free
 }
 {
  bone1Name=dyn_l_sword_chain_05_reverse
  bone2Name=backpack_mstr
  swing1LimitDegrees=45
  swing1Motion=Free
  swing2LimitDegrees=45
  swing2Motion=Free
  twistLimitDegrees=45
  twistMotion=Free
 }
 {
  bone1Name=dyn_r_sword_chain_05_reverse
  bone2Name=backpack_mstr
  swing1LimitDegrees=60
  swing1Motion=Free
  swing2LimitDegrees=60
  swing2Motion=Free
  twistLimitDegrees=5
  twistMotion=Free
 }
 {bone1Name=dyn_l_sword_chain_05, bone2Name=dyn_l_sword_chain_04, swing1LimitDegrees=60, swing2LimitDegrees=60, twistLimitDegrees=5}
 {
  bone1Name=dyn_l_sword_chain_05_reverse
  bone2Name=dyn_l_sword_chain_05
  swing1LimitDegrees=45
  swing1Motion=Free
  swing2LimitDegrees=45
  swing2Motion=Free
  twistLimitDegrees=45
  twistMotion=Free
 }
 {bone1Name=dyn_r_sword_chain_05, bone2Name=dyn_r_sword_chain_04, swing1LimitDegrees=60, swing2LimitDegrees=60, twistLimitDegrees=5}
 {
  bone1Name=dyn_r_sword_chain_05_reverse
  bone2Name=dyn_r_sword_chain_05
  swing1LimitDegrees=45
  swing1Motion=Free
  swing2LimitDegrees=45
  swing2Motion=Free
  twistLimitDegrees=45
  twistMotion=Free
 }
 {
  bone1Name=dyn_r_sword_chain_05_reverse
  bone2Name=dyn_r_sword_chain_04
  swing1LimitDegrees=45
  swing1Motion=Free
  swing2LimitDegrees=45
  swing2Motion=Free
  twistLimitDegrees=45
  twistMotion=Free
 }
 {
  bone1Name=dyn_r_sword_chain_05_reverse
  bone2Name=dyn_r_sword_chain_03
  swing1LimitDegrees=45
  swing1Motion=Free
  swing2LimitDegrees=45
  swing2Motion=Free
  twistLimitDegrees=45
  twistMotion=Free
 }
 {
  bone1Name=dyn_r_sword_chain_05_reverse
  bone2Name=dyn_r_sword_chain_02
  swing1LimitDegrees=45
  swing1Motion=Free
  swing2LimitDegrees=45
  swing2Motion=Free
  twistLimitDegrees=45
  twistMotion=Free
 }
 {
  bone1Name=dyn_l_sword_chain_05_reverse
  bone2Name=dyn_l_sword_chain_04
  swing1LimitDegrees=45
  swing1Motion=Free
  swing2LimitDegrees=45
  swing2Motion=Free
  twistLimitDegrees=45
  twistMotion=Free
 }
 {
  bone1Name=dyn_l_sword_chain_05_reverse
  bone2Name=dyn_l_sword_chain_03
  swing1LimitDegrees=45
  swing1Motion=Free
  swing2LimitDegrees=45
  swing2Motion=Free
  twistLimitDegrees=45
  twistMotion=Free
 }
 {
  bone1Name=dyn_l_sword_chain_05_reverse
  bone2Name=dyn_l_sword_chain_02
  swing1LimitDegrees=45
  swing1Motion=Free
  swing2LimitDegrees=45
  swing2Motion=Free
  twistLimitDegrees=45
  twistMotion=Free
 }
 {
  bone1Name=dyn_l_sword_blade_base
  bone2Name=backpack_mstr
  swing1LimitDegrees=45
  swing1Motion=Free
  swing2LimitDegrees=45
  swing2Motion=Free
  twistLimitDegrees=45
  twistMotion=Free
 }
]
