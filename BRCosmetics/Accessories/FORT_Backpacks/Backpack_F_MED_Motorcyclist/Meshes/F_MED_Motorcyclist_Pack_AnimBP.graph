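# Animation Blueprint for the Motorcyclist backpack cosmetic, adding secondary motion to its zipper and sword chains.
# A RigidBody node simulates from the reference pose in root bone space, then Anim Dynamics swings the right zipper.
# Constraints pin each chain's fifth bone to its reverse bone, while the third-bone constraints apply only during skydiving.
# The event graph copies bIsSkyDiving into EG_Sky_Diving every update to toggle those constraints.
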AnimBlueprint /BRCosmetics/Accessories/FORT_Backpacks/Backpack_F_MED_Motorcyclist/Meshes/F_MED_Motorcyclist_Pack_AnimBP
target_skeleton=&/F_MED_Motorcyclist_Pack_Skeleton
graphs [
 AnimGraph {nodes [
   n1 = RigidBody(ComponentPose=Local To Component(LocalPose=Local Space Ref Pose())) {
    bOverrideWorldGravity=true
    overridePhysicsAsset=&/F_MED_Motorcyclist_Pack_Physics
    simulationSpace=BaseBoneSpace
    alphaBoolBlend {blendOption=Linear}
    baseBoneRef {boneName=root}
    componentAppliedLinearAccClamp=<3000 3000 3000>
    componentLinearAccScale=<0.3 0.3 0.45>
    componentLinearVelScale=<0.15 0.15 0.225>
    overrideWorldGravity {z=-980}
   }
   n2 = Anim Dynamics - Bone: dyn_r_zipper_base(ComponentPose=@n1, Alpha=1, ExternalForce="0,0,0") {
    angularDampingOverride=0.999
    bGravityOverrideInSimSpace=true
    bOverrideAngularDamping=true
    simulationSpace=RootRelative
    alphaBoolBlend {blendOption=Linear}
    boundBone {boneName=dyn_r_zipper_base}
    componentLinearAccScale=<1 1 1>
    componentLinearVelScale=<1 1 1>
    physicsBodyDefinitions [{
      sphereCollisionRadius=0
      boundBone {boneName=dyn_r_zipper_base}
      boxExtents=<3 10 10>
      constraintSetup {angularLimitsMax=<2 0 45>, angularLimitsMin=<-2 -60 -5>, angularTarget=<0 0 0>}
      localJointOffset=<1.5 -0.05 0.1>
     }]
   }
   n3 = "Constraint - dyn_l_sword_chain_05 (P)"(ComponentPose=@n2, Alpha=1, ConstraintWeights_0=1) {
    alphaBoolBlend {blendOption=Linear}
    boneToModify {boneName=dyn_l_sword_chain_05}
    constraintSetup [{targetBone {boneName=dyn_l_sword_chain_05_reverse}}]
    constraintWeights [1]
   }
   n4 = "Constraint - dyn_r_sword_chain_05 (P)"(ComponentPose=@n3, Alpha=1, ConstraintWeights_0=1) {
    alphaBoolBlend {blendOption=Linear}
    boneToModify {boneName=dyn_r_sword_chain_05}
    constraintSetup [{targetBone {boneName=dyn_r_sword_chain_05_reverse}}]
    constraintWeights [1]
   }
   n5 = Get EG_Sky_Diving() -> (EG_Sky_Diving:bool)
   n6 = "Constraint - dyn_r_sword_chain_03 (P)"!(ComponentPose=@n4, bAlphaBoolEnabled:bool=@n5, ConstraintWeights_0=1) -> (Pose) {
    alphaInputType=Bool
    alphaBoolBlend {blendOption=Linear}
    boneToModify {boneName=dyn_r_sword_chain_03}
    constraintSetup [{targetBone {boneName=dyn_r_sword_chain_03_constraint}}]
    constraintWeights [1]
   }
   n7 = "Constraint - dyn_l_sword_chain_03 (P)"!(ComponentPose=@n6, bAlphaBoolEnabled:bool=@n5, ConstraintWeights_0=1) -> (Pose) {
    alphaInputType=Bool
    alphaBoolBlend {blendOption=Linear}
    boneToModify {boneName=dyn_l_sword_chain_03}
    constraintSetup [{targetBone {boneName=dyn_l_sword_chain_03_constraint}}]
    constraintWeights [1]
   }
   Output Pose(Result=Component To Local(ComponentPose=@n7)) {layerGroup=None}
  ]}
 EventGraph {nodes [
   Event BlueprintUpdateAnimation()
   > n1 = Branch(Condition=Get bIsSkyDiving())
   > Set EG_Sky_Diving(EG_Sky_Diving:bool=true) -> (Output_Get:bool)
   Set EG_Sky_Diving(execute=@n1.else, EG_Sky_Diving=false)
   TryGetPawnOwner()
  ]}
]
variables [EG_Sky_Diving {type=boolean}]
